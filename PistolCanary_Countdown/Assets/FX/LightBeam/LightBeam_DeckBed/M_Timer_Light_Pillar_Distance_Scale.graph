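# Additive, unlit, two-sided light pillar material for meshes and particles that grows with camera distance to stay readable from afar.
# Particle color and time-of-day brightness drive emissive; opacity combines gradients, depth and near-camera fades, and optional sphere, linear and end masks.
# Spline thickening and position-scaling WPO follow a distance ramp, with width from dynamic or scalar parameters.
# Switches allow a controlled BeaconLocation, which inverts that ramp and fades opacity, and bunker-pillar distance fading.
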Material /PistolCanary_Countdown/Assets/FX/LightBeam/LightBeam_DeckBed/M_Timer_Light_Pillar_Distance_Scale
graph {
 properties {
  bUsedWithMeshParticles=true
  bUsedWithNanite=true
  bUsedWithNiagaraMeshParticles=true
  bUsedWithNiagaraRibbons=true
  bUsedWithNiagaraSprites=true
  bUsedWithParticleSprites=true
  blendMode=BLEND_Additive
  pixelDepthOffsetMode=PDOM_Legacy
  shadingModel=MSM_Unlit
  twoSided=true
  usedShadingModels=Unlit
 }
 nodes [
  n1 = ParticleColor()
  n2 = ScalarParameter() {defaultValue=1, parameterName=BrightnessMultiMorning}
  n3 = ScalarParameter() {defaultValue=1, parameterName=BrightnessMultiDay}
  n4 = ScalarParameter() {defaultValue=1, parameterName=BrightnessMultiEvening}
  n5 = ScalarParameter() {defaultValue=1, parameterName=BrightnessMultiNight}
  MaterialFunctionCall(Emissive=Multiply(
    A=@n1.RGB
    B=MaterialFunctionCall(Value_Morning=@n2, Value_Day=@n3, Value_Evening=@n4, Value_Night=@n5) {materialFunction=&Scalar_Time_of_Day}.Result
   )) {materialFunction=&MF_FX_EmissiveAdjustments}
  n6 = ScalarParameter() {defaultValue=0.2, parameterName=Additional Distance Fade Ofs}
  n7 = Subtract(
   A=MaterialFunctionCall(Distance=ScalarParameter() {defaultValue=4000, parameterName=Additional Distance Fade}) {materialFunction=&Near_Camera_Fade}.Result
   B=@n6
  )
  n8 = Saturate(None=Multiply(A=Saturate(None=@n7), B=Add(A=@n6)))
  n9 = Multiply(
   A=TextureCoordinate()
   B=AppendVector(
    A=ScalarParameter() {defaultValue=1, group=SphereMask, parameterName=SphereMask-UV-X}
    B=ScalarParameter() {defaultValue=0.3, group=SphereMask, parameterName=SphereMask-UV-Y}
   )
  )
  n10 = ComponentMask(None=@n9) {g=true}
  n11 = ScalarParameter() {group=SphereMask, parameterName=Spheremask UV.Y Min}
  n12 = ScalarParameter() {defaultValue=0.5, group=SphereMask, parameterName=Spheremask UV.Y Max}
  n13 = StaticSwitchParameter(True=AppendVector(A=ComponentMask(None=@n9) {r=true}, B=Clamp(None=@n10, Min=@n11, Max=@n12) {maxDefault=0.5}), False=@n9) {
   desc=Clamps spheremask uv Y component. Use to make spheremask only affect one portion of the length of the beam
   group=SphereMask
   parameterName=Clamp Spheremask UV.Y
  }
  n14 = SphereMask(
   A=@n13
   B=AppendVector(
    A=ScalarParameter() {defaultValue=0.5, group=SphereMask, parameterName=SphereMask-CenterX}
    B=ScalarParameter() {defaultValue=-0.225, group=SphereMask, parameterName=SphereMask-CenterY}
   )
   Radius=ScalarParameter() {defaultValue=0.55, group=SphereMask, parameterName=SphereMask-Radius}
   Hardness=ScalarParameter() {group=SphereMask, parameterName=SphereMask-Hardness}
  )
  n15 = StaticSwitchParameter(True=@n1.A, False=VectorParameter() {parameterName=BaseColor, defaultValue=<rgba 1 1 1 1>}.A) {defaultValue=true, parameterName=UseWParticles}
  n16 = TextureCoordinate() {coordinateIndex=1}
  n17 = MaterialFunctionCall() {materialFunction=&LinearGradient}
  n18 = Multiply(
   A=Power(Base=TextureSample(UVs=@n16) {samplerType=SAMPLERTYPE_LinearColor, texture=/Game/Global/Textures/Default/Masks_and_Gradients/T_VertGradient}.R)
   B=Power(Base=OneMinus(None=@n17.VGradient), Exp=ScalarParameter() {defaultValue=6, parameterName=GradientFadePower}) {constExponent=6}
  )
  n19 = ScalarParameter() {defaultValue=300, parameterName=DepthFadeDistance}
  n20 = ScalarParameter() {defaultValue=1, parameterName=DepthFade-NearFadeBasedMultiplier}
  n21 = MaterialFunctionCall(Distance=ScalarParameter() {defaultValue=200, parameterName=NearFadeDistance}) {materialFunction=&Near_Camera_Fade}
  n22 = DepthFade(Opacity=@n18, FadeDistance=LinearInterpolate(A=@n19, B=Multiply(A=@n19, B=@n20), Alpha=@n21.Result)) {fadeDistanceDefault=200}
  n23 = Multiply(
   A=Multiply(A=StaticSwitchParameter(True=@n1.A, False=@n15) {parameterName=for Bunker Pillar}, B=FeatureLevelSwitch(Default=@n22, ES2=@n18, ES3_1=@n18))
   B=@n21.Result
  )
  n24 = ObjectPositionWS()
  n25 = Distance(
   A=StaticSwitchParameter(True=VectorParameter() {parameterName=BeaconLocation}.RGB, False=@n24) {parameterName=Use Controlled Distance Vectors for Scaling}
   B=CameraPositionWS()
  )
  n26 = ScalarParameter() {defaultValue=2000, parameterName=DistanceOffset}
  n27 = Clamp(None=Divide(A=Subtract(A=@n25, B=@n26), B=ScalarParameter() {defaultValue=30000, parameterName=ScaleUpDistance}))
  n28 = StaticSwitchParameter(True=OneMinus(None=@n27), False=@n27) {parameterName=Use Controlled Distance Vectors for Scaling}
  n29 = Multiply(
   A=ScalarParameter() {defaultValue=1, parameterName=Opacity}
   B=StaticSwitchParameter(True=LinearInterpolate(B=@n23, Alpha=@n28), False=@n23) {parameterName=Use Controlled Distance Vectors for Scaling}
  )
  n30 = MaterialFunctionCall() {materialFunction=&LinearGradient}
  n31 = OneMinus(None=@n30.VGradient)
  n32 = Multiply(
   A=Power(Base=Multiply(A=@n30.VGradient, B=@n31), Exp=ScalarParameter() {defaultValue=1, group=Fade Top Bottom, parameterName=Fade Top Bottom Power})
   B=ScalarParameter() {defaultValue=1, group=Fade Top Bottom, parameterName=Fade Top Bottom Multiplier}
  )
  n33 = LinearInterpolate(A=@n29, B=Multiply(A=@n29, B=Clamp(None=@n32)), Alpha=ScalarParameter() {group=Fade Top Bottom, parameterName=Fade Top Bottom})
  n34 = TextureCoordinate()
  n35 = Multiply(A=Power(Base=ComponentMask(None=@n34) {g=true}, Exp=ScalarParameter() {defaultValue=1, parameterName=LinearMask_Range}) {constExponent=1}) {constB=4}
  n36 = Multiply(
   A=LinearInterpolate(A=@n33, B=Multiply(A=@n33, B=@n14), Alpha=ScalarParameter() {group=SphereMask, parameterName=UseSphereMask, sortPriority=1})
   B=LinearInterpolate(B=Saturate(None=@n35), Alpha=ScalarParameter() {parameterName=UseLinearMask}) {constA=1}
  )
  Multiply(A=StaticSwitchParameter(True=Multiply(A=@n8, B=@n36), False=@n36) {parameterName=for Bunker Pillar}, B=NamedRerouteUsage())
  n37 = DynamicParameter() {desc=WPO Only, defaultValue=<rgba 100 1 1 1>, paramNames [SplineWidth, Na, Master Opac, NA]}
  NamedRerouteDeclaration(None=@n37.SplineWidth) {name=DP0_R Spline Witdh, nodeColor=<rgba 0 0.3672 1 1>}
  NamedRerouteDeclaration(None=@n37.Master Opac) {name=DP0_B MasterOpac, nodeColor=<rgba 0 0 1 1>}
  n38 = ScalarParameter() {defaultValue=60, parameterName=MaxScaleIncrease}
  n39 = WorldPosition()
  n40 = LinearInterpolate(B=Multiply(A=Multiply(A=@n38) {constB=-1}, B=Subtract(A=@n24, B=@n39.XYZ)) {constB=-1}, Alpha=@n28) {constB=0}
  n41 = Constant()
  n42 = ScalarParameter() {group=Camera Distance Spline Scale, parameterName=Spline Scale Camera Distance Min}
  n43 = ScalarParameter() {defaultValue=1, group=Camera Distance Spline Scale, parameterName=Spline Scale Camera Distance Max, sortPriority=33}
  n44 = ScalarParameter() {defaultValue=1, parameterName=SplineWidth}
  n45 = Multiply(
   A=Max(A=MaterialFunctionCall(Input=@n21.Result, Target Low=@n42, Target High=@n43) {materialFunction=&RemapValueRange}.Result) {constB=0}
   B=@n44
  )
  n46 = ScalarParameter() {defaultValue=20, parameterName=MaxWidthMult}
  n47 = Multiply(
   A=StaticSwitchParameter(True=NamedRerouteUsage(), False=StaticSwitchParameter(True=@n45, False=@n44) {parameterName=Enable Spline Scale With Camera Fade}) {defaultValue=true, parameterName=UseWParticles}
   B=StaticSwitchParameter(True=LinearInterpolate(B=@n46, Alpha=@n28) {constA=1, constB=3}, False=Constant() {r=1}) {defaultValue=true, parameterName=UseScaleUpDistance}
  )
  Reroute(None=Add(
    A=MaterialFunctionCall(WidthBase=@n47, WidthTip=@n47) {materialFunction=&SplineThicken}.WorldPositionOffset
    B=StaticSwitchParameter(True=@n40, False=@n41) {defaultValue=true, parameterName=UseScaleUpDistance}
   ))
 ]
}
