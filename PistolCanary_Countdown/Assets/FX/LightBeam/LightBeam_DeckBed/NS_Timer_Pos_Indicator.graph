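# Niagara system for a timer position indicator: one persistent amber mesh beam that, going by its name, marks a timer's location from afar.
# A single burst spawns a looping particle, rotated 90 degrees and nudged by a camera offset.
# Camera distance drives curves that stretch the mesh's local X axis and feed a dynamic material parameter.
# LOD distance toggles the visibility tag and fades alpha in as the viewer moves away, hiding the beam up close.
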NiagaraSystem /PistolCanary_Countdown/Assets/FX/LightBeam/LightBeam_DeckBed/NS_Timer_Pos_Indicator
dependencies {
 usedDataInterfaces [&NiagaraDataInterfaceCamera, &NiagaraDataInterfaceCurve]
 usedDynamicInputs []
 usedModules [&SystemState]
}
emitters {PickupBeam {
  emitterUpdateScript [
   &EmitterState()
   &SpawnBurst_Instantaneous(Age:NiagaraFloat=$Emitter.Age, Spawn Count=1, Use Spawn Probability=false)
   /SetVariables(Emitter.Cam Pos:NiagaraPosition=&ReturnCameraPosition(Coordinate Space:ENiagaraCoordinateSpace=::Simulation, Camera Query:NiagaraDataInterfaceCamera=DataInterface())) {isSetParameters=true}
  ]
  particleSpawnScript [
   &InitializeParticle(
    Color=<rgba 1.5 0.5995 0.0781 0.35>
    Lifetime=4
    Mesh Scale Mode=::NewEnumerator4
    Sprite Size Mode=::NewEnumerator3
    Uniform Sprite Size=8693.0547
    UsePositionOffset=false
    Write Material Random=true
   ) {module=InitializeParticle001}
   /SetVariables(Particles.MeshOrientation=&MakeQuatFromAxisAngle(Axis=<0 -1 0>, AngleInDegrees=90)) {isSetParameters=true}
   &CameraOffset(Camera Offset Amount=-20)
   /SetVariables(Particles.Color mul=0.5) {isSetParameters=true}
  ]
  particleUpdateScript [
   &ParticleState(
    Kill Particles When Lifetime Has Elapsed=false
    Let Infinitely Lived Particles Die When Emitter Deactivates=true
    Loop Particles Lifetime=true
   )
   &ScaleMeshSize(Scale Factor=&MakeVector(
     X=&FloatFromCurve(
      FloatCurve=DataInterface() {propertyValues {Curve {keys [<at 2000 1>,<at 30000 60>,<at 1000000 1000>]}}}
      CurveIndex=&VectorLength(Vector=&Subtract_Position(A=$Particles.Position, B=$Emitter.Cam Pos))
     )
     Y=1
     Z=1
    ))
   &DynamicMaterialParameters(
    Index 0 Param 1=&FloatFromCurve(
     FloatCurve=DataInterface() {propertyValues {Curve {keys [<at 2000 1000>,<at 30000 1000>,<at 500000 25000>]}}}
     CurveIndex=&VectorLength(Vector=&Subtract_Position(A=$Particles.Position, B=$Emitter.Cam Pos))
     Scale Curve=0.5
    )
    Index 0 Param 3=0.35
    Index 0 Param 4=1
    Module.Index0Param0Write=true
    Module.Index0Param1Write=true
    Module.Index0Param2Write=true
    Module.Index0Param3Write=true
   )
   /SetVariables(Particles.VisibilityTag=&MakeCustomIntFromBool(Bool=&SetBoolByFloatComparison(A=$Engine.Owner.LODDistance, B=22000), True INT=0, False INT=1)) {isSetParameters=true}
   &Color(
    Color=$Particles.Initial.Color
    Scale Alpha=&Lerp_Float(
     A=1
     B=&FloatFromCurve(
      FloatCurve=DataInterface() {propertyValues {Curve {keys [<at 0.45 0>,<at 1 1>]}}}
      CurveIndex=&Divide_Float(A=$Engine.Owner.LODDistance, B=50000)
     )
     Alpha=1
    )
    Scale Color=&VectorFromFloat(Value=$Particles.Color mul)
   )
  ]
  rendererClasses [&NiagaraMeshRendererProperties]
  renderers [{rendererClass=&NiagaraMeshRendererProperties, rendererIndex=0}]
 }}
summary {
 systemName=NS_Timer_Pos_Indicator
 emitters [{emitterName=PickupBeam, rendererClasses [&NiagaraMeshRendererProperties]}]
 userVariables [User.UseFlicker {type=NiagaraFloat, defaultValue=0, description=""}]
}
